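# Niagara system for the player storm-immune effect, drawing pulsing flashes attached to the character's skeletal mesh.
# Its single Decal_Flash emitter bursts nine persistent particles, each pinned by partition index to a bone or socket with its own size and offset.
# A system waveform and a per-particle, index-staggered waveform drive purple-to-orange color, alpha, size and dynamic material parameters.
# Separate parameter bindings feed two sprite renderers and a mesh renderer.
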NiagaraSystem /Game/Effects/Fort_Effects/Effects/Characters/Generic/StormImmune/NS_Player_StormImmune
dependencies {
 usedDataInterfaces [&NiagaraDataInterfaceArrayFloat, &NiagaraDataInterfaceCurve]
 usedDynamicInputs []
 usedModules [
  &self:SystemSpawnScript.SystemScriptSource.SystemScriptGraph.NiagaraNodeAssignment_0.SetVariables_#1
  &self:SystemSpawnScript.SystemScriptSource.SystemScriptGraph.NiagaraNodeAssignment_1.SetVariables_#2
  &SystemState
  /Niagara/Noise/Waveform
  &self:SystemSpawnScript.SystemScriptSource.SystemScriptGraph.NiagaraNodeAssignment_2.SetVariables_#3
 ]
}
emitters {Decal_Flash {
  emitterUpdateScript [
   &EmitterState(Life Cycle Mode=::NewEnumerator1, Loop Behavior=::NewEnumerator1, Loop Duration Mode=::NewEnumerator1, UseLoopDelay=false)
   &SpawnBurst_Instantaneous(Spawn Count=9, Use Spawn Probability=false, UseLoopCountLimit=true)
  ]
  particleSpawnScript [
   &PartitionParticles(Number Of Partitions=9)
   &InitializeParticle(
    Sprite Size Mode=::NewEnumerator3
    Uniform Sprite Size=&SelectFloatFromArray(
     Direct Array Index=$Particles.PartitionParticles.Partition
     Float Selection Array=DataInterface() {propertyValues {FloatData [60,50,50,60,60,50,60,60,60]}}
    )
    UsePositionOffset=false
    Write Material Random=true
   )
  ]
  particleUpdateScript [
   &ParticleState(Kill Particles When Lifetime Has Elapsed=false, Let Infinitely Lived Particles Die When Emitter Deactivates=true)
   /SetVariables(Particles.ParticleSequence:NiagaraFloat=&Power(
     a=&ClampFloat(Float=/Niagara/DynamicInputs/Helpers/Waveform(
       Global Input Position=&Add_Float(A=0, B=&ReturnNormalizedExecIndex(Normalized Index Scale=0.2))
       "Waveform [1]"=::NewEnumerator0
       "[1] Frequency"=1
       "[1] Amplitude Min/Max"=<-4 1>
      ))
     b=2
    )) {isSetParameters=true}
   &CameraOffset(Camera Offset Amount=10)
   /SetVariables(Particles.CameraOffset_P:NiagaraFloat=&Lerp_Float(A=10, B=20, Alpha=$Particles.ParticleSequence)) {isSetParameters=true}
   &SkeletalMeshLocation(
    Bone / Socket Index=$Particles.PartitionParticles.Partition
    Bone Sampling Mode=::NewEnumerator4
    Inherit Velocity=false
    MaskSpawnGroup=false
    Sampled Position Offset=&Multiply_VectorByFloat(
     Vector=<1 0 0>
     Float=&SelectFloatFromArray(
      Direct Array Index=$Particles.PartitionParticles.Partition
      Float Selection Array=DataInterface() {propertyValues {FloatData [5,0,0,0,-5,20,-20,20,-20]}}
     )
    )
    Sampled Position Offset Coordinate Space:ENiagaraMeshTransforms=::NewEnumerator3
    Skeletal Mesh=$System.SkeletalMesh
   )
   &ScaleColor(
    Scale Alpha=&Lerp_Float(A=0.3, B=0, Alpha=$System.Waveform.Waveform)
    Scale RGB=&VectorFromFloat(Value=&FNTimeOfDayScalar(Night Scale=3, Morning Scale=6, Day Scale=10, Evening Scale=6))
    ScaleA=true
    ScaleRGB=true
   ) {enabled=false, module=ScaleColor001}
   &DynamicMaterialParameters(
    Index 0 Param 1=1
    Index 0 Param 2=&Lerp_Float(A=2, B=10, Alpha=$System.WaveformPower)
    Index 2 Param 1=1
    Index 2 Param 2=&Lerp_Float(A=1, B=6, Alpha=$Particles.ParticleSequence)
    Module.Index0Param0Write=true
    Module.Index0Param1Write=true
    Module.Index0Param2Write=true
    Module.Index0Param3Write=true
    Module.Index2Param0Write=true
    Module.Index2Param1Write=true
    Module.Index2Param2Write=true
    Module.Index2Param3Write=true
    Write Parameter Index 2=true
   )
   &ScaleColor(
    Scale Alpha=&Lerp_Float(A=0.25, B=1, Alpha=$System.WaveformPower)
    Scale RGB=&Lerp_Vector(A=<0.5 0 2>, B=<4 2 0>, Alpha=$System.WaveformPower)
    ScaleA=true
    ScaleRGB=true
   )
   /SetVariables(Particles.Color_P:LinearColor=&MultiplyLinearColor(
     Linear Color=<rgba 3 3 3 0.15>
     Scale Factor=&MakeFloat4FromVectorAndFloat(
      "Vector (RGB)"=&Lerp_Vector(
       A=<0.25 0 1>
       B=<1 0.5 0>
       Alpha=&FloatFromCurve(
        FloatCurve=DataInterface() {propertyValues {Curve {keys [<at 0.75 0 RCIM_Cubic>, {tangentMode=RCTM_User, time=0.8, value=1}]}}}
        CurveIndex=$Particles.ParticleSequence
       )
      )
      "Float (Alpha)"=$Particles.ParticleSequence
     )
    )) {isSetParameters=true}
   /SetVariables(Particles.SpriteSize_P:Vector2f=&Multiply_Vector2D_ByFloat(Float=&Lerp_Float(A=0.5, B=0.6, Alpha=$Particles.ParticleSequence), Vector2D=$Particles.SpriteSize)) {isSetParameters=true}
  ]
  rendererClasses [&NiagaraSpriteRendererProperties, &NiagaraMeshRendererProperties]
  renderers [
   {rendererClass=&NiagaraSpriteRendererProperties}
   {rendererClass=&NiagaraSpriteRendererProperties, rendererIndex=1}
   {rendererClass=&NiagaraMeshRendererProperties, rendererIndex=2}
  ]
 }}
summary {
 systemName=NS_Player_StormImmune
 emitters [{emitterName=Decal_Flash, rendererClasses [&NiagaraSpriteRendererProperties, &NiagaraMeshRendererProperties]}]
 userVariables [
  User.IntensityScalar {type=NiagaraFloat, defaultValue=0.4, description=""}
  User.Skeletal Mesh {
   type=NiagaraDataInterfaceSkeletalMesh
   defaultValue {
    struct=NiagaraExt_VariableValue_DataInterface
    value {dataInterface=&self:NiagaraDataInterfaceSkeletalMesh_0}
   }
  }
 ]
}
